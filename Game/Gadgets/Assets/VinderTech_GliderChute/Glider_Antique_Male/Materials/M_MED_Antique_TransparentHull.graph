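# Translucent, unlit fake-glass material for the transparent hull of the Antique Male glider.
# Emissive blends the diffuse texture toward a color parameter, masked by normal-map deviation from flat plus a hand-built specular glint reflecting the global sunlight vector.
# Opacity starts from a base value, optionally adds a local-Z bounds gradient via static switch, and turns opaque at the glint.
# Attributes pass through the accessories TechArt override function.
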Material /Game/Gadgets/Assets/VinderTech_GliderChute/Glider_Antique_Male/Materials/M_MED_Antique_TransparentHull
graph {
 properties {
  bCastRayTracedShadows=false
  bUseMaterialAttributes=true
  bUsedWithMeshDeformer=true
  bUsedWithSkeletalMesh=true
  blendMode=BLEND_Translucent
  pixelDepthOffsetMode=PDOM_Legacy
  shadingModel=MSM_Unlit
  usedShadingModels=Unlit
 }
 nodes [
  n1 = TextureSample() {texture=/Game/Gadgets/Assets/VinderTech_GliderChute/Glider_Antique_Male/Textures/T_Glider_Antique_Male_D}
  n2 = VectorParameter() {defaultValue=<rgba 1 1 1 0>}
  n3 = TextureSample() {samplerType=SAMPLERTYPE_Normal, texture=&FlatNormalBlue}
  n4 = TextureSample() {samplerType=SAMPLERTYPE_Normal, texture=/Game/Gadgets/Assets/VinderTech_GliderChute/Glider_Antique_Male/Textures/T_Glider_Antique_Male_N}
  n5 = Subtract(A=@n3.RGB, B=@n4.RGB)
  n6 = CollectionParameter() {
   collection=&FortniteMaterialParameters
   parameterName=SunlightDirectionalVector
  }
  n7 = ComponentMask(None=@n6) {b=true, g=true, r=true}
  n8 = PixelNormalWS()
  n9 = Multiply(A=DotProduct(A=@n7, B=@n8)) {constB=2}
  n10 = Normalize(VectorInput=Subtract(A=Multiply(A=@n9, B=@n8), B=@n7))
  n11 = Multiply(A=MaterialFunctionCall() {materialFunction=&CameraDirectionVector}.Result) {constB=-1}
  n12 = Power(Base=Max(A=DotProduct(A=@n10, B=@n11)) {constB=0}, Exp=ScalarParameter() {defaultValue=20, group=FakeGlass, parameterName=SpecPow})
  n13 = Add(
   A=Multiply(
    A=MaterialFunctionCall(Vector 3=@n5) {materialFunction=/Engine/Functions/Engine_MaterialFunctions02/Utility/VectorLength}.V3 Length
    B=ScalarParameter() {defaultValue=0.3, parameterName=NormalEdgingMask, sliderMax=1}
   )
   B=@n12
  )
  n14 = ScalarParameter() {defaultValue=0.4, parameterName=BaseTransparency, sliderMax=1}
  n15 = ComponentMask(None=MaterialFunctionCall() {materialFunction=&LocalPosition}."Local Position (Excluding Offsets)") {b=true}
  n16 = MaterialFunctionCall() {materialFunction=&ObjectLocalBounds}
  n17 = ComponentMask(None=@n16.Local Bounds Minimum) {b=true}
  n18 = Power(
   Base=Divide(A=Subtract(A=@n15, B=@n17), B=ComponentMask(None=@n16.Local Bounds Size) {b=true})
   Exp=ScalarParameter() {defaultValue=1, parameterName=GradientPow}
  ) {constExponent=1.5}
  n19 = ScalarParameter() {defaultValue=50, parameterName=GradientMultiply}
  n20 = StaticSwitchParameter(True=Add(A=@n14, B=Multiply(A=@n18, B=@n19) {constB=50}), False=@n14) {parameterName=ApplyZGradiant}
  MaterialFunctionCall(In=SetMaterialAttributes!(MaterialAttributes, Emissive Color=LinearInterpolate(A=@n1.RGB, B=@n2.RGB, Alpha=@n13), Opacity=LinearInterpolate(A=@n20, Alpha=@n12)) {attributeSetTypes [#1, #2]}) {materialFunction=&MF_TechArtOverrides_accessories}
  ObjectPositionWS()
  OneMinus(None=Power(Base=DotProduct(A=@n3.RGB, B=@n4.RGB), Exp=ScalarParameter() {defaultValue=10, parameterName=NormalEdgingPower}))
  VertexNormalWS()
 ]
}
