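# Skinned material for the Antique male glider that fakes glass in masked areas of a character-style surface.
# Character PreFX, with optional cloth fuzz and silk, blends by Glass_Mask into a glass layer with sun specular, fresnel emissive and frontend color overrides, then feeds PostFX.
# A dithered opacity term combines two-sided sign and specular, turning solid for shadows.
# World-offset logic adds a switchable camera-based shoulder fix and a sun-directed shadow offset.
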Material /Game/Gadgets/Assets/VinderTech_GliderChute/Glider_Antique_Male/Materials/M_MED_FakeGlass_Antique
graph {
 properties {
  bAutomaticallySetUsageInEditor=false
  bCastRayTracedShadows=false
  bDisableDepthTest=true
  bEnableResponsiveAA=true
  bUseMaterialAttributes=true
  bUsedWithMeshDeformer=true
  bUsedWithMorphTargets=true
  bUsedWithSkeletalMesh=true
  ditherOpacityMask=true
  pixelDepthOffsetMode=PDOM_Legacy
  shadingModel=MSM_FromMaterialExpression
  subsurfaceProfile=&SS_HeroSkin_02
  translucencyDirectionalLightingIntensity=5
  translucencyLightingMode=TLM_Surface
  usedShadingModels=Default Lit | Subsurface Profile | Cloth
 }
 nodes [
  n1 = CollectionParameter() {
   collection=&FortniteMaterialParameters
   parameterName=SunlightDirectionalVector
  }
  n2 = ComponentMask(None=@n1) {b=true, g=true, r=true}
  n3 = PixelNormalWS()
  n4 = Multiply(A=Multiply(A=DotProduct(A=@n2, B=@n3)) {constB=2}, B=@n3)
  DotProduct(
   A=Normalize(VectorInput=Subtract(A=@n4, B=@n2))
   B=Multiply(A=MaterialFunctionCall() {materialFunction=&CameraDirectionVector}.Result)
  )
  n5 = MaterialFunctionCall() {materialFunction=&BaseFN_CharacterFunction_PreFX}
  n6 = TextureSampleParameter2D() {parameterName=SkinFX_Mask}
  n7 = ChannelMaskParameter(None=@n6.RGBA) {group=Skin FX, parameterName=Silk Channel}
  n8 = StaticSwitchParameter(
   True=MaterialFunctionCall!(
    Alpha=StaticComponentMaskParameter(None=@n6.RGBA) {defaultR=true, group=Skin FX, parameterName=Cloth Channel}
    Gmap=@n5.Gmap
    Material=@n5.Gmap
   ) {materialFunction=/Game/Weapons/WeaponSkins/MaterialFunction/ClothFuzz}.Result
   False=@n5.Result
  ) {group=Skin FX, parameterName=Use Cloth fuzz}
  n9 = GetMaterialAttributes(None=StaticSwitchParameter(
    True=MaterialFunctionCall!(Alpha=@n7.R, Material=@n8) {materialFunction=/Game/Weapons/WeaponSkins/MaterialFunction/MF_Silk}.Result
    False=@n8
   ) {group=Skin FX, parameterName=Use Silk}) {attributeGetTypes [#1]}
  n10 = VectorParameter() {group=FrontendHacks, parameterName=FrontendColor, defaultValue=<rgba 0.1636 0.3458 0.8021 1>}
  n11 = VectorParameter() {group=FrontendHacks, parameterName=FrontendColorFwdRenderer, defaultValue=<rgba 0.0878 0.2948 0.4531 1>}
  n12 = VectorParameter() {group=FakeGlass, parameterName=FakeGlassBaseColor, defaultValue=<rgba 1 1 1 0>}
  n13 = Multiply(A=TwoSidedSign(), B=ScalarParameter() {defaultValue=1, group=FakeGlass, parameterName=SignNorm})
  n14 = CollectionParameter() {
   collection=&FortniteMaterialParameters
   parameterName=SunlightDirectionalVector
  }
  n15 = ComponentMask(None=@n14) {b=true, g=true, r=true}
  n16 = PixelNormalWS()
  n17 = Multiply(A=Multiply(A=DotProduct(A=@n15, B=@n16)) {constB=2}, B=@n16)
  n18 = DotProduct(
   A=Normalize(VectorInput=Subtract(A=@n17, B=@n15))
   B=Multiply(A=MaterialFunctionCall() {materialFunction=&CameraDirectionVector}.Result) {constB=-1}
  )
  n19 = Multiply(A=OneMinus(None=@n13), B=Power(Base=Max(A=@n18) {constB=0}, Exp=ScalarParameter() {defaultValue=20, group=FakeGlass, parameterName=SpecPow}))
  n20 = OneMinus(None=@n19)
  n21 = ScalarParameter() {group=FrontendHacks, parameterName=IsInFrontEnd}
  n22 = ScalarParameter() {defaultValue=3, group=FrontendHacks, parameterName=Fresnel-Power}
  n23 = StaticBool() {value=true}
  n24 = ShadingPathSwitch(
   Default=ScalarParameter() {defaultValue=0.2, group=FrontendHacks, parameterName=Fresnel-DarkLevel}
   Forward=ScalarParameter() {group=FrontendHacks, parameterName=Fresnel-DarkLevelFWD}
  )
  n25 = ShadingPathSwitch(
   Default=ScalarParameter() {defaultValue=1, group=FrontendHacks, parameterName=Fresnel-BrightLevel}
   Forward=ScalarParameter() {defaultValue=1, group=FrontendHacks, parameterName=Fresnel-BrightLevelFWD}
  )
  n26 = Saturate(None=Add(
    A=OneMinus(None=@n21)
    B=MaterialFunctionCall(Power=@n22, Use Cheap contrast=@n23, Cheap contrast dark=@n24, Cheap contrast bright=@n25) {materialFunction=&Fresnel_Function}.Result
   ))
  n27 = Normalize(VectorInput=Multiply(A=Transform(None=@n9.Normal), B=Constant3Vector() {constant=<rgba 1 1 -1 0>}) {constB=-1})
  n28 = OneMinus(None=@n26)
  n29 = VectorParameter() {group=FrontendHacks, parameterName=FrontendEmissiveFwdRenderer, defaultValue=<rgba 0.0106 0.0225 0.0521 1>}
  n30 = ShadingPathSwitch(
   Default=VectorParameter() {group=FrontendHacks, parameterName=FrontendEmissive, defaultValue=<rgba 0.0563 0.119 0.276 1>}.RGB
   Forward=@n29.RGB
  )
  n31 = Multiply(
   A=CollectionParameter() {collection=&FortniteMaterialParameters, parameterName=SunLightColor}
   B=ScalarParameter() {defaultValue=2.5, group=FakeGlass, parameterName=SpecIntensity}
  )
  n32 = Add(
   A=ComponentMask(None=Multiply(A=@n19, B=@n31)) {b=true, g=true, r=true}
   B=VectorParameter() {group=FakeGlass, parameterName=BaseEmissive}.RGB
  )
  n33 = MaterialFunctionCall(Normal Vector=PixelNormalWS(), Power=ScalarParameter() {defaultValue=3, parameterName=FresnelPower}) {materialFunction=&Fresnel_Function}
  n34 = Multiply(A=@n33.Result, B=VectorParameter() {group=FakeGlass, parameterName=FresnelColor, defaultValue=<rgba 0.4896 0.4372 0.1555 0>}.RGB)
  n35 = SetMaterialAttributes!(
   MaterialAttributes=@n9.MaterialAttributes
   Base Color=LinearInterpolate(A=ShadingPathSwitch(Default=@n10.RGB, Forward=@n11.RGB), B=Multiply(A=@n12.RGB, B=@n20), Alpha=@n26)
   Metallic=Multiply(A=@n26, B=ScalarParameter() {defaultValue=0.8, group=FakeGlass, parameterName=FakeGlassMetallic})
   Specular=Multiply(A=@n26, B=ScalarParameter() {group=FakeGlass, parameterName=FakeGlassSpecular})
   Roughness=ScalarParameter() {defaultValue=0.32, group=FakeGlass, parameterName=FakeGlassRoughness}
   Normal=StaticSwitchParameter(True=Transform(None=@n27) {transformSourceType=TRANSFORMSOURCE_World, transformType=TRANSFORM_Tangent}, False=@n9.Normal) {group=FakeGlass, parameterName=FlipNormalZ}
   Emissive Color=Add(A=Multiply(A=@n28, B=@n30), B=Add(A=@n32, B=@n34))
  ) {attributeSetTypes [
    #2
    #3
    #4
    #5
    #1
    #6
   ]}
  n36 = TextureSampleParameter2D() {
   group=FakeGlass
   parameterName=Glass_Mask
   samplerType=SAMPLERTYPE_LinearGrayscale
   texture=/Game/Characters/Player/Male/Medium/Bodies/M_MED_Holiday_Lights/Textures/T_M_MED_Holiday_Lights_Body_G
  }
  MaterialFunctionCall(
   AO=@n5.AO
   Distance Blend=@n5.AO
   Fresnel=@n5.AO
   Gmap=@n5.AO
   Material=BlendMaterialAttributes(A=@n9.MaterialAttributes, B=@n35, Alpha=@n36.R)
  ) {materialFunction=&BaseFN_CharacterFunction_PostFX}
  Multiply(A=@n33.Result, B=ScalarParameter() {defaultValue=0.5, parameterName=FresnelIntensity})
  n37 = Multiply(A=@n19, B=ScalarParameter() {defaultValue=1, group=FakeGlass, parameterName=SpecOpacity}) {constB=0.4}
  n38 = ShadowReplace(Default=Saturate(None=Add(A=@n13, B=@n37)), Shadow=Constant() {r=1})
  RayTracingQualitySwitch(Normal=@n38, RayTraced=Multiply(A=@n38, B=ScalarParameter() {defaultValue=1, group=FakeGlass, parameterName=RTOpacityMaskMul}))
  n39 = CollectionParameter() {collection=&FortniteMaterialParameters, parameterName=PlayerPosition}
  n40 = Multiply(
   A=Distance(A=ComponentMask(None=@n39) {b=true, g=true, r=true}, B=ActorPositionWS())
   B=ScalarParameter() {defaultValue=0.01, group=ShoulderFix, parameterName=DistThresh}
  )
  n41 = OneMinus(None=Round(None=Saturate(None=@n40)))
  n42 = MaterialFunctionCall() {materialFunction=&CameraDirectionVector}
  n43 = LinearInterpolate(
   A=ScalarParameter() {defaultValue=20, group=ShoulderFix, parameterName=z}
   B=ScalarParameter() {defaultValue=9, group=ShoulderFix, parameterName=zv}
   Alpha=Abs(None=ComponentMask(None=@n42.Result) {b=true})
  )
  n44 = WorldPosition()
  n45 = CameraPositionWS()
  n46 = WorldPosition()
  n47 = ScalarParameter() {defaultValue=140, group=ShoulderFix, parameterName=rad}
  n48 = ScalarParameter() {defaultValue=1, group=ShoulderFix, parameterName=hard}
  n49 = Multiply(
   A=Multiply(
    A=MaterialFunctionCall(Offset Amount=@n43, World Position=@n44.XYZ) {materialFunction=/Engine/Functions/Engine_MaterialFunctions02/WorldPositionOffset/CameraOffset}.Camera Offset
    B=SphereMask(A=@n45, B=@n46.XYZ, Radius=@n47, Hardness=@n48)
   )
   B=SphereMask(
    A=PreSkinnedPosition()
    B=VectorParameter() {group=ShoulderFix, parameterName=s, defaultValue=<rgba -25 0 128 0>}.RGB
    Radius=ScalarParameter() {defaultValue=12, group=ShoulderFix, parameterName=rad2}
    Hardness=ScalarParameter() {defaultValue=1, group=ShoulderFix, parameterName=hard2}
   )
  )
  n50 = CollectionParameter() {
   collection=&FortniteMaterialParameters
   parameterName=SunAndMoonModelWorldPosition
  }
  n51 = Subtract(A=ComponentMask(None=@n50) {b=true, g=true, r=true}, B=WorldPosition().XYZ)
  n52 = ScalarParameter() {parameterName=ShadowOffset}
  n53 = ScalarParameter() {group=FrontendHacks, parameterName=ShadowOffsetFrontend}
  n54 = ScalarParameter() {group=FrontendHacks, parameterName=IsInFrontEnd}
  ShadowReplace(
   Default=StaticSwitchParameter(True=Multiply(A=@n41, B=@n49) {constA=1}, False=Constant()) {parameterName=ShoulderFix}
   Shadow=Multiply(A=Normalize(VectorInput=@n51), B=LinearInterpolate(A=@n52, B=@n53, Alpha=@n54))
  )
  VertexNormalWS()
 ]
}
